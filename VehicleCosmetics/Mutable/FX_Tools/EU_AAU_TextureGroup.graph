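# Editor utility for bulk-editing selected assets: sets texture groups on textures and text fields on vehicle boost and drift trail cosmetics.
# Basics functions write an entered name and description to boosters or drift trails, falling back to variant tokens when that cast fails.
# Short functions write short descriptions from default string variables, with separate functions for painted-style variant tokens.
# Each function loops over the selection inside an undo transaction, then checks out and saves dirty assets.
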EditorUtilityBlueprint /VehicleCosmetics/Mutable/FX_Tools/EU_AAU_TextureGroup
graphs [
 Change Texture Group {nodes [
   n1 = Change Texture Group!() -> (then, Texture Group:TextureGroup Enum) {is=title K2Node_FunctionEntry}
   > BeginTransaction(Context=Change Texture Group)
   > n2 = GetSelectedAssets()
   > n3 = For Each Loop(Array:Array of Object References=@n2) -> (Array Element:Object Reference)
   CheckoutLoadedAssets(execute=@n3.Completed, AssetsToCheckout=@n2)
   > SaveLoadedAssets(AssetsToSave=@n2, bOnlyIfIsDirty=true)
   > EndTransaction()
   n4 = Cast To Texture!(execute=@n3.LoopBody, Object=@n3) -> (then, CastFailed, AsTexture:Texture Object Reference) {is=title K2Node_DynamicCast}
   > TransactObject(Object=@n4)
   > Set LODGroup!(execute, LODGroup:TextureGroup Enum=@n1, self:Texture Object Reference=@n4) -> (then, Output_Get:TextureGroup Enum)
  ]}
 EventGraph {links [], nodes []}
 Set Boost Basics {nodes [
   n1 = Set Boost Basics!() -> (then, Name:Text, Description:Text) {is=title K2Node_FunctionEntry}
   > BeginTransaction(Context=Set Short Description)
   > n2 = GetSelectedAssets()
   > n3 = For Each Loop(Array=@n2)
   CheckoutLoadedAssets(execute=@n3.Completed, AssetsToCheckout=@n2)
   > SaveLoadedAssets(AssetsToSave=@n2, bOnlyIfIsDirty=true)
   > EndTransaction()
   n4 = Cast To FortVehicleCosmeticsItemDefinition_Booster!(execute=@n3.LoopBody, Object=@n3) -> (then, CastFailed, AsFort Vehicle Cosmetics Item Definition Booster:Fort Vehicle Cosmetics Item Definition Booster Object Reference) {is=title K2Node_DynamicCast}
   > TransactObject(Object=@n4)
   > SetEditorProperty(Object=@n4, PropertyName=ItemName, PropertyValue:Text=@n1)
   > SetEditorProperty(Object=@n4, PropertyName=ItemDescription, PropertyValue=@n1.Description)
   n5 = Cast To FortVehicleCosmeticsVariantTokenType!(execute=@n4.CastFailed, Object=@n3) -> (then, CastFailed, AsFort Vehicle Cosmetics Variant Token Type:Fort Vehicle Cosmetics Variant Token Type Object Reference) {is=title K2Node_DynamicCast}
   > TransactObject(Object=@n5)
   > SetEditorProperty(Object=@n5, PropertyName=ItemName, PropertyValue=@n1)
   > SetEditorProperty(Object=@n5, PropertyName=ItemDescription, PropertyValue=@n1.Description)
  ]}
 Set Boost Short {nodes [
   Set Boost Short() {is=title K2Node_FunctionEntry}
   > BeginTransaction(Context=Set Short Description)
   > n1 = GetSelectedAssets()
   > n2 = For Each Loop(Array=@n1)
   CheckoutLoadedAssets(execute=@n2.Completed, AssetsToCheckout=@n1)
   > SaveLoadedAssets(AssetsToSave=@n1, bOnlyIfIsDirty=true)
   > EndTransaction()
   n3 = Cast To FortVehicleCosmeticsItemDefinition_Booster(execute=@n2.LoopBody, Object=@n2)
   > TransactObject(Object=@n3)
   > SetEditorProperty(Object=@n3, PropertyName=ItemShortDescription, PropertyValue=Get BoostDescription() -> (BoostDescription:Text))
  ]}
 Set Boost Variant Short {nodes [
   Set Boost Variant Short() {is=title K2Node_FunctionEntry}
   > BeginTransaction(Context=Set Short Description)
   > n1 = GetSelectedAssets()
   > n2 = For Each Loop(Array=@n1)
   CheckoutLoadedAssets(execute=@n2.Completed, AssetsToCheckout=@n1)
   > SaveLoadedAssets(AssetsToSave=@n1, bOnlyIfIsDirty=true)
   > EndTransaction()
   n3 = Cast To FortVehicleCosmeticsVariantTokenType(execute=@n2.LoopBody, Object=@n2)
   > TransactObject(Object=@n3)
   > SetEditorProperty(Object=@n3, PropertyName=ItemShortDescription, PropertyValue=Get BoostVariantDescription() -> (BoostVariantDescription:Text))
  ]}
 Set Trail Basics {nodes [
   n1 = Set Trail Basics!() -> (then, Name:Text, Description:Text) {is=title K2Node_FunctionEntry}
   > BeginTransaction(Context=Set Short Description)
   > n2 = GetSelectedAssets()
   > n3 = For Each Loop(Array=@n2)
   CheckoutLoadedAssets(execute=@n3.Completed, AssetsToCheckout=@n2)
   > SaveLoadedAssets(AssetsToSave=@n2, bOnlyIfIsDirty=true)
   > EndTransaction()
   n4 = Cast To FortVehicleCosmeticsItemDefinition_DriftTrail!(execute=@n3.LoopBody, Object=@n3) -> (then, CastFailed, AsFort Vehicle Cosmetics Item Definition Drift Trail:Fort Vehicle Cosmetics Item Definition Drift Trail Object Reference) {is=title K2Node_DynamicCast}
   > TransactObject(Object=@n4)
   > SetEditorProperty(Object=@n4, PropertyName=ItemName, PropertyValue=@n1)
   > SetEditorProperty(Object=@n4, PropertyName=ItemDescription, PropertyValue=@n1.Description)
   n5 = Cast To FortVehicleCosmeticsVariantTokenType(execute=@n4.CastFailed, Object=@n3)
   > TransactObject(Object=@n5)
   > SetEditorProperty(Object=@n5, PropertyName=ItemName, PropertyValue=@n1)
   > SetEditorProperty(Object=@n5, PropertyName=ItemDescription, PropertyValue=@n1.Description)
  ]}
 Set Trail Short {nodes [
   Set Trail Short() {is=title K2Node_FunctionEntry}
   > BeginTransaction(Context=Set Short Description)
   > n1 = GetSelectedAssets()
   > n2 = For Each Loop(Array=@n1)
   CheckoutLoadedAssets(execute=@n2.Completed, AssetsToCheckout=@n1)
   > SaveLoadedAssets(AssetsToSave=@n1, bOnlyIfIsDirty=true)
   > EndTransaction()
   n3 = Cast To FortVehicleCosmeticsItemDefinition_DriftTrail(execute=@n2.LoopBody, Object=@n2)
   > TransactObject(Object=@n3)
   > SetEditorProperty(Object=@n3, PropertyName=ItemShortDescription, PropertyValue=Get TrailDescription() -> (TrailDescription:Text))
  ]}
 Set Trail Variant Short {nodes [
   Set Trail Variant Short() {is=title K2Node_FunctionEntry}
   > BeginTransaction(Context=Set Short Description)
   > n1 = GetSelectedAssets()
   > n2 = For Each Loop(Array=@n1)
   CheckoutLoadedAssets(execute=@n2.Completed, AssetsToCheckout=@n1)
   > SaveLoadedAssets(AssetsToSave=@n1, bOnlyIfIsDirty=true)
   > EndTransaction()
   n3 = Cast To FortVehicleCosmeticsVariantTokenType(execute=@n2.LoopBody, Object=@n2)
   > TransactObject(Object=@n3)
   > SetEditorProperty(Object=@n3, PropertyName=ItemShortDescription, PropertyValue=Get TrailVariantDescription() -> (TrailVariantDescription:Text))
  ]}
]
variables [
 BoostDescription {type=string, default=Boost}
 BoostVariantDescription {type=string, default=Boost Painted Style}
 TrailDescription {type=string, default=Trail}
 TrailVariantDescription {type=string, default=Trail Painted Style}
]
